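# Animation Blueprint for the GrimeHold male character head, adding fin physics and a Control Rig pass.
# The input pose gets an additive head counter-rotation, scaled from its delta to neck_02, then chained Anim Dynamics on front, center and back fins.
# A Control Rig then takes targeting, emote, pawn and frontend state.
# The computed gravity override is unused by the fixed-gravity fin nodes, and the cached Blinks blend never reaches the output.
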AnimBlueprint /BRCosmetics/Characters/Player/Male/Medium/Heads/M_MED_GrimeHold_Head/Meshes/CustomSkeleton/Fortnite_Base_Head_GrimeHold_Skeleton_AnimB
target_skeleton=&/Fortnite_Base_Head_GrimeHold_Skeleton
graphs [
 AnimGraph {nodes [
   n1 = Control Rig!(
    Source=Use cached pose 'startPose'()
    Alpha
    Is_targeting:bool=Get bIsTargeting()
    Is_emote:bool=Get bIsPlayingEmote()
    isPawn:bool=Get isPawn() -> (isPawn:bool)
    Custom Idle is Playing Float:double=Get Custom Idle is Playing Float() -> (Custom Idle is Playing Float:double)
    Is_frontend:bool=Get IsFrontend() -> (IsFrontend:bool)
   ) -> (Pose) {controlRigAssetReference {blueprintRigClass=/BRCosmetics/Characters/Player/Male/Medium/Heads/M_MED_GrimeHold_Head/Meshes/M_MED_GrimeHold_Head_CtrlRig_base.M_MED_GrimeHold_Head_CtrlRig_base_C}}
   Output Pose(Result=@n1)
   Save cached pose 'Blinks'(Pose=Layered blend per bone(BlendPoses_0=Slot 'DefaultSlot'(), BlendWeights_0=Get Custom Idle is Playing Float()) {layerSetup [{branchFilters [{boneName=faceAttach}]}]}) {cacheName=Blinks}
   n2 = Input Pose()
   n3 = Anim Dynamics - Bone: dyn_head_fin_front_01!(
    ComponentPose="Transform (Modify) Bone - Bone: head"(ComponentPose=Local To Component(LocalPose=@n2), Alpha=1, Translation="0,0,0", Rotation=Get head rot() -> (head rot:Rotator), Scale="1,1,1") {rotationMode=BMM_Additive, rotationSpace=BCS_BoneSpace, boneToModify {boneName=head}}
    Alpha=1
    GravityOverride:Vector="0,0,-980"
   ) -> (Pose) {
    angularDampingOverride=1
    angularSpringConstant=10000
    bAngularSpring=true
    bOverrideAngularDamping=true
    bUseGravityOverride=true
    gravityScale=0.35
    simulationSpace=RootRelative
    alphaBoolBlend {blendOption=Linear}
    boundBone {boneName=dyn_head_fin_front_01}
    componentAppliedLinearAccClamp=<3000 3000 3000>
    componentLinearAccScale=<0.1 0.1 0.1>
    componentLinearVelScale=<0.1 0.1 0.1>
    gravityOverride {z=-980}
    physicsBodyDefinitions [{
      boundBone {boneName=dyn_head_fin_front_01}
      boxExtents=<14.5 24.5 24.5>
      constraintSetup {angularLimitsMax=<15 15 15>, angularLimitsMin=<-15 -15 -15>, angularTarget=<1 0 0>}
      localJointOffset=<0.8 0 0>
     }]
   }
   n4 = Anim Dynamics - Bone: dyn_head_fin_front_02!(ComponentPose=@n3, Alpha=1, GravityOverride:Vector="0,0,-980") -> (Pose) {
    angularDampingOverride=1
    angularSpringConstant=10000
    bAngularSpring=true
    bOverrideAngularDamping=true
    bUseGravityOverride=true
    gravityScale=0.35
    simulationSpace=RootRelative
    alphaBoolBlend {blendOption=Linear}
    boundBone {boneName=dyn_head_fin_front_02}
    componentAppliedLinearAccClamp=<3000 3000 3000>
    componentLinearAccScale=<0.1 0.1 0.1>
    componentLinearVelScale=<0.1 0.1 0.1>
    gravityOverride {z=-980}
    physicsBodyDefinitions [{
      boundBone {boneName=dyn_head_fin_front_02}
      boxExtents=<14.5 24.5 24.5>
      constraintSetup {angularLimitsMax=<15 15 8>, angularLimitsMin=<-15 -15 -8>, angularTarget=<1 0 0>}
      localJointOffset=<0.8 0 0>
     }]
   }
   n5 = Anim Dynamics - Bone: dyn_head_fin_center_01!(ComponentPose=@n4, Alpha=1, GravityOverride:Vector="0,0,-980") -> (Pose) {
    angularDampingOverride=1
    angularSpringConstant=10000
    bAngularSpring=true
    bOverrideAngularDamping=true
    bUseGravityOverride=true
    gravityScale=0.35
    simulationSpace=RootRelative
    alphaBoolBlend {blendOption=Linear}
    boundBone {boneName=dyn_head_fin_center_01}
    componentAppliedLinearAccClamp=<3000 3000 3000>
    componentLinearAccScale=<0.1 0.1 0.1>
    componentLinearVelScale=<0.1 0.1 0.1>
    gravityOverride {z=-980}
    physicsBodyDefinitions [{
      boundBone {boneName=dyn_head_fin_center_01}
      boxExtents=<15 25 25>
      constraintSetup {angularLimitsMax=<15 15 15>, angularLimitsMin=<-15 -15 -15>, angularTarget=<1 0 0>}
      localJointOffset=<0.8 0 0>
     }]
   }
   n6 = Anim Dynamics - Bone: dyn_head_fin_center_02!(ComponentPose=@n5, Alpha=1, GravityOverride:Vector="0,0,-980") -> (Pose) {
    angularDampingOverride=1
    angularSpringConstant=10000
    bAngularSpring=true
    bOverrideAngularDamping=true
    bUseGravityOverride=true
    gravityScale=0.35
    simulationSpace=RootRelative
    alphaBoolBlend {blendOption=Linear}
    boundBone {boneName=dyn_head_fin_center_02}
    componentAppliedLinearAccClamp=<3000 3000 3000>
    componentLinearAccScale=<0.1 0.1 0.1>
    componentLinearVelScale=<0.1 0.1 0.1>
    gravityOverride {z=-980}
    physicsBodyDefinitions [{
      boundBone {boneName=dyn_head_fin_center_02}
      boxExtents=<15 25 25>
      constraintSetup {angularLimitsMax=<15 15 8>, angularLimitsMin=<-15 -15 -8>, angularTarget=<1 0 0>}
      localJointOffset=<0.8 0 0>
     }]
   }
   n7 = Anim Dynamics - Bone: dyn_head_fin_bck_01!(ComponentPose=@n6, Alpha=1, GravityOverride:Vector="0,0,-980") -> (Pose) {
    angularDampingOverride=1
    angularSpringConstant=10000
    bAngularSpring=true
    bOverrideAngularDamping=true
    bUseGravityOverride=true
    gravityScale=0.35
    simulationSpace=RootRelative
    alphaBoolBlend {blendOption=Linear}
    boundBone {boneName=dyn_head_fin_bck_01}
    componentAppliedLinearAccClamp=<3000 3000 3000>
    componentLinearAccScale=<0.1 0.1 0.1>
    componentLinearVelScale=<0.1 0.1 0.1>
    gravityOverride {z=-980}
    physicsBodyDefinitions [{
      boundBone {boneName=dyn_head_fin_bck_01}
      boxExtents=<15.5 25.5 25.5>
      constraintSetup {angularLimitsMax=<15 15 15>, angularLimitsMin=<-15 -15 -15>, angularTarget=<1 0 0>}
      localJointOffset=<0.8 0 0>
     }]
   }
   n8 = Anim Dynamics - Bone: base_head_fin_bck_02!(ComponentPose=@n7, Alpha=1, GravityOverride:Vector="0,0,-980") -> (Pose) {
    angularDampingOverride=1
    angularSpringConstant=10000
    bAngularSpring=true
    bOverrideAngularDamping=true
    bUseGravityOverride=true
    gravityScale=0.35
    simulationSpace=RootRelative
    alphaBoolBlend {blendOption=Linear}
    boundBone {boneName=base_head_fin_bck_02}
    componentAppliedLinearAccClamp=<3000 3000 3000>
    componentLinearAccScale=<0.1 0.1 0.1>
    componentLinearVelScale=<0.1 0.1 0.1>
    gravityOverride {z=-980}
    physicsBodyDefinitions [{
      boundBone {boneName=base_head_fin_bck_02}
      boxExtents=<15.5 25.5 25.5>
      constraintSetup {angularLimitsMax=<15 15 8>, angularLimitsMin=<-15 -15 -8>, angularTarget=<1 0 0>}
      localJointOffset=<0.8 0 0>
     }]
   }
   Save cached pose 'startPose'(Pose=Component To Local(ComponentPose=@n8)) {cacheName=startPose}
   Use cached pose 'Blinks'()
  ]}
 EventGraph {nodes [
   Event BlueprintUpdateAnimation()
   > n1 = Sequence!(execute) -> (then_0, then_1, then_2:Exec, then_3:Exec)
   > n2 = "Gravity Override State Function (Nativized)"!(
    execute
    CharacterPartAnimInstance:Custom Character Part Anim Instance Object Reference=Self-Reference()
    GravityOverrideParameterSet:Fort Gravity Override Parameter Set Structure=Get Gravity Override Params_0() -> (Gravity Override Params_0:Fort Gravity Override Parameter Set Structure)
    RigidBodyAnimNodeParameterSet:Fort Rigid Body Anim Node Parameter Set Structure=Get RBAN Params_0() -> (RBAN Params_0:Fort Rigid Body Anim Node Parameter Set Structure)
    DebugDraw:bool=false
    GameTimeInSeconds:float=GetGameTimeInSeconds()
    DebugJointName:Name=root
    StateLogicLayerName:Name=none
   ) -> (
    then
    OutGravityOverride:Vector
    OutCurrentState:Name=None
    OutGravityOverrideParameters:Fort Gravity Override Parameters Structure
    OutRigidBodyAnimNodeParameters:Fort Rigid Body Anim Node Parameters Structure
   )
   > n3 = Set Gravity Override(Gravity Override=@n2)
   > Set Gravity Override Cape(Gravity Override Cape:Vector=@n3) -> (Output_Get:Vector)
   > Set RBAN Settings Out_0(RBAN Settings Out_0:Fort Rigid Body Anim Node Parameters Structure=@n2.OutRigidBodyAnimNodeParameters) -> (Output_Get:Fort Rigid Body Anim Node Parameters Structure)
   n4 = GetOwnerSkeletalMeshComponent()
   n5 = BreakRotator(InRot=BreakTransform(InTransform=GetDeltaTransformFromRefPose(self=@n4, BoneName=head, BaseName=neck_02)).Rotation)
   n6 = Get Y percentage() -> (Y percentage:float)
   n7 = Get X percentage() -> (X percentage:float)
   n8 = Get Z percentage() -> (Z percentage:float)
   > Set head rot(head rot:Rotator=MakeRotator(
     Roll=float * float!(A=@n5, B=-1, C:double=@n7) -> (ReturnValue)
     Pitch=float * float(A=@n5.Pitch, B=-1, C=@n6)
     Yaw=float * float(A=@n5.Yaw, B=-1, C=@n8)
    )) -> (Output_Get:Rotator)
   Branch(execute=@n1.then_2, Condition=Get bIsInFrontEnd())
   > Set Custom Idle is Playing Float(Custom Idle is Playing Float:double=GetCurveValue(self=GetOwnerFrontEndAnimInstance(), CurveName=Frontend_Idle_Bypass)) -> (Output_Get:double)
   n9 = Is Valid(exec=@n1.then_3, InputObject=TryGetPawnOwner())
   Set IsFrontend(execute=@n9.Is Not Valid, IsFrontend:bool=false) -> (Output_Get:bool)
   Set IsFrontend(execute=@n9.Is Valid, IsFrontend=true)
   n10 = Is Valid(exec=@n1.then_1, InputObject=TryGetPawnOwner())
   Set isPawn(execute=@n10.Is Not Valid, isPawn:bool=false) -> (Output_Get:bool)
   Set isPawn(execute=@n10.Is Valid, isPawn=true)
   TryGetPawnOwner()
   Set PlayingEmoteOnly(PlayingEmoteOnly:bool=Get bIsPlayingEmote()) -> (Output_Get:bool)
  ]}
]
variables [
 Custom Idle is Playing Float {type=number}
 Gravity Override {type=Vector}
 Gravity Override Cape {type=Vector}
 Gravity Override Params_0 {
  type=FortGravityOverrideParameterSet
  default=(Skydive_GoingUp=(GravityOverride=()),Skydive_FloatingDown=(GravityOverride=()),Skydive_DivingDown=(GravityOverride=()),Skydive_Parachuting=(GravityOverride=()),OnGround_Standing=(GravityOverride=()),OnGround_Moving=(GravityOverride=()),OnGround_Moving_HighRotationRate=(GravityOverride=()),OnGround_Crouch_Moving=(GravityOverride=()),Emote_or_Melee=(GravityOverride=()),Swimming=(GravityOverride=()),Front_End=(GravityOverride=()),DBNO=(GravityOverride=()),DBNO_Carried=(GravityOverride=()),OnGround_Falling=(GravityOverride=()),Ziplining=(GravityOverride=()),Skydive_Surfing=(GravityOverride=()),Riding_Standing=(GravityOverride=()),Riding_Moving=(GravityOverride=()),Riding_Moving_Targeting=(GravityOverride=()),Tactical_Sprint=(GravityOverride=()),Rail_Grinding=(GravityOverride=()),)
 }
 head rot {type=Rotator}
 Is Targeting {type=boolean}
 IsFrontend {type=boolean}
 isPawn {type=boolean}
 PlayingEmoteOnly {type=boolean}
 RBAN Params_0 {
  type=FortRigidBodyAnimNodeParameterSet
  default=(Skydive_GoingUp=(RigidBodyAnimNode=()),Skydive_FloatingDown=(RigidBodyAnimNode=()),Skydive_DivingDown=(RigidBodyAnimNode=()),Skydive_Parachuting=(RigidBodyAnimNode=()),OnGround_Standing=(RigidBodyAnimNode=()),OnGround_Moving=(RigidBodyAnimNode=()),OnGround_Moving_HighRotationRate=(RigidBodyAnimNode=()),OnGround_Crouch_Moving=(RigidBodyAnimNode=()),Emote_or_Melee=(RigidBodyAnimNode=()),Swimming=(RigidBodyAnimNode=()),Front_End=(RigidBodyAnimNode=()),DBNO=(RigidBodyAnimNode=()),DBNO_Carried=(RigidBodyAnimNode=()),OnGround_Falling=(RigidBodyAnimNode=()),Ziplining=(RigidBodyAnimNode=()),Skydive_Surfing=(RigidBodyAnimNode=()),Riding_Standing=(RigidBodyAnimNode=()),Riding_Moving=(RigidBodyAnimNode=()),Riding_Moving_Targeting=(RigidBodyAnimNode=()),Tactical_Sprint=(RigidBodyAnimNode=()),Rail_Grinding=(RigidBodyAnimNode=()),)
 }
 RBAN Settings Out_0 {type=FortRigidBodyAnimNodeParameters}
 X percentage {type=number, default=0.5}
 Y percentage {type=number, default=0.5}
 Z percentage {type=number, default=0.5}
]
